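# Physics asset for the hair (HIS part) of the female medium Zombie Elastic player cosmetic, apparently for secondary hair motion.
# Capsules on head, neck, upper spine and upper arms form the body volumes, joined by free-angle constraints.
# Five hair chains branch from the hair root through free-angle base bodies: three-link A and two-link B pairs, plus a back chain.
# Strand links have tight angular limits, with boxes capping the B and back chains.
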PhysicsAsset /BRCosmetics/Characters/Player/Female/Medium/Bodies/F_MED_Zombie_Elastic/Meshes/Parts/HIS/F_MED_Zombie_Elastic_HIS_Physics
bodies {
 head [{shapeType=Capsule, length=4.25, radius=5.75, center=<1.1782 3.0621 0>, rotation=<rot 90 0 0>}]
 his1_hair [{shapeType=Capsule, length=1, radius=1}]
 his1_hair_A_a_base_l [{shapeType=Capsule, length=0.5, radius=0.5}]
 his1_hair_A_a_base_r [{shapeType=Capsule, length=0.5, radius=0.5}]
 his1_hair_A_a_l [{shapeType=Capsule, length=2.5, radius=1.75, center=<3 0 0>, rotation=<rot 90 0 0>}]
 his1_hair_A_a_r [{shapeType=Capsule, length=2.25, radius=2, center=<-3 0 0>, rotation=<rot 90 0 0>}]
 his1_hair_A_b_l [{shapeType=Capsule, length=2.75, radius=1.5, center=<3 0 0>, rotation=<rot 90 0 0>}]
 his1_hair_A_b_r [{shapeType=Capsule, length=2.5, radius=1.75, center=<-3 0 0>, rotation=<rot 90 0 0>}]
 his1_hair_A_c_l [{shapeType=Capsule, length=2.5, radius=1.75, center=<3 0 0>, rotation=<rot 90 0 0>}]
 his1_hair_A_c_r [{shapeType=Capsule, length=3.25, radius=1.75, center=<-4 0 0>, rotation=<rot 90 0 0>}]
 his1_hair_B_a_base_l [{shapeType=Capsule, length=0.5, radius=0.5}]
 his1_hair_B_a_base_r [{shapeType=Capsule, length=0.5, radius=0.5}]
 his1_hair_B_a_l [{shapeType=Capsule, length=1.25, radius=3, center=<3 0 0>, rotation=<rot 90 0 0>}]
 his1_hair_B_a_r [{shapeType=Capsule, length=1.75, radius=2.5, center=<3 0 0>, rotation=<rot 90 0 0>}]
 his1_hair_B_b_l [{shapeType=Box, extentX=3.1835, extentY=4.9768, extentZ=11.3535, shapeName=his1_hair_B_b_l_box, center=<3.4021 0 0>, rotation=<rot 0 90 -90>}]
 his1_hair_B_b_r [{shapeType=Box, extentX=5.0491, extentY=3.1211, extentZ=10.4143, shapeName=his1_hair_B_b_r_box, center=<5.0534 0 0>, rotation=<rot 0 90 90>}]
 his1_hair_bk_a [{shapeType=Capsule, length=0, radius=4.75, center=<3 0 0>, rotation=<rot 90 0 0>}]
 his1_hair_bk_a_base_l [{shapeType=Capsule, length=0.5, radius=0.5}]
 his1_hair_bk_b [{shapeType=Box, extentX=4.1468, extentY=8.8318, extentZ=13.4431, shapeName=his1_hair_bk_b_box, center=<4.8273 0 0>, rotation=<rot 0 90 90>}]
 neck_01 [{shapeType=Capsule, length=7.75, radius=5, center=<2.2564 0.5078 0>, rotation=<rot 90 0 0>}]
 spine_05 [{shapeType=Capsule, length=16.75, radius=9.25, center=<1.5648 0.0001 0>}]
 upperarm_l [{shapeType=Capsule, length=15.5, radius=4, center=<9.1076 0 0>, rotation=<rot 90 0 0>}]
 upperarm_r [{shapeType=Capsule, length=15.5, radius=4, center=<-9.1076 0 0>, rotation=<rot 90 0 180>}]
}
constraints [
 {
  bone1Name=his1_hair
  bone2Name=head
  swing1LimitDegrees=45
  swing1Motion=Free
  swing2LimitDegrees=45
  swing2Motion=Free
  twistLimitDegrees=45
  twistMotion=Free
 }
 {
  bone1Name=head
  bone2Name=neck_01
  swing1LimitDegrees=45
  swing1Motion=Free
  swing2LimitDegrees=45
  swing2Motion=Free
  twistLimitDegrees=45
  twistMotion=Free
 }
 {
  bone1Name=neck_01
  bone2Name=spine_05
  swing1LimitDegrees=45
  swing1Motion=Free
  swing2LimitDegrees=45
  swing2Motion=Free
  twistLimitDegrees=45
  twistMotion=Free
 }
 {
  bone1Name=his1_hair_bk_a_base_l
  bone2Name=his1_hair
  swing1LimitDegrees=45
  swing1Motion=Free
  swing2LimitDegrees=45
  swing2Motion=Free
  twistLimitDegrees=45
  twistMotion=Free
 }
 {
  bone1Name=his1_hair_B_a_base_l
  bone2Name=his1_hair
  swing1LimitDegrees=45
  swing1Motion=Free
  swing2LimitDegrees=45
  swing2Motion=Free
  twistLimitDegrees=45
  twistMotion=Free
 }
 {
  bone1Name=his1_hair_A_a_base_r
  bone2Name=his1_hair
  swing1LimitDegrees=45
  swing1Motion=Free
  swing2LimitDegrees=45
  swing2Motion=Free
  twistLimitDegrees=45
  twistMotion=Free
 }
 {
  bone1Name=his1_hair_B_a_base_r
  bone2Name=his1_hair
  swing1LimitDegrees=45
  swing1Motion=Free
  swing2LimitDegrees=45
  swing2Motion=Free
  twistLimitDegrees=45
  twistMotion=Free
 }
 {
  bone1Name=his1_hair_A_a_base_l
  bone2Name=his1_hair
  swing1LimitDegrees=45
  swing1Motion=Free
  swing2LimitDegrees=45
  swing2Motion=Free
  twistLimitDegrees=45
  twistMotion=Free
 }
 {bone1Name=his1_hair_bk_a, bone2Name=his1_hair_bk_a_base_l, swing1LimitDegrees=10, swing2LimitDegrees=10, twistLimitDegrees=10}
 {bone1Name=his1_hair_bk_b, bone2Name=his1_hair_bk_a, swing1LimitDegrees=30, swing2LimitDegrees=30, twistLimitDegrees=20}
 {bone1Name=his1_hair_B_a_l, bone2Name=his1_hair_B_a_base_l, swing1LimitDegrees=10, swing2LimitDegrees=10, twistLimitDegrees=10}
 {bone1Name=his1_hair_B_b_l, bone2Name=his1_hair_B_a_l, swing1LimitDegrees=30, swing2LimitDegrees=30, twistLimitDegrees=20}
 {bone1Name=his1_hair_A_a_r, bone2Name=his1_hair_A_a_base_r, swing1LimitDegrees=15, swing2LimitDegrees=15, twistLimitDegrees=15}
 {bone1Name=his1_hair_A_b_r, bone2Name=his1_hair_A_a_r, swing1LimitDegrees=15, swing2LimitDegrees=15, twistLimitDegrees=15}
 {bone1Name=his1_hair_B_a_r, bone2Name=his1_hair_B_a_base_r, swing1LimitDegrees=10, swing2LimitDegrees=10, twistLimitDegrees=10}
 {bone1Name=his1_hair_B_b_r, bone2Name=his1_hair_B_a_r, swing1LimitDegrees=30, swing2LimitDegrees=30, twistLimitDegrees=20}
 {bone1Name=his1_hair_A_a_l, bone2Name=his1_hair_A_a_base_l, swing1LimitDegrees=15, swing2LimitDegrees=15, twistLimitDegrees=15}
 {bone1Name=his1_hair_A_b_l, bone2Name=his1_hair_A_a_l, swing1LimitDegrees=15, swing2LimitDegrees=15, twistLimitDegrees=15}
 {bone1Name=his1_hair_A_c_r, bone2Name=his1_hair_A_b_r, swing1LimitDegrees=15, swing2LimitDegrees=15, twistLimitDegrees=15}
 {bone1Name=his1_hair_A_c_l, bone2Name=his1_hair_A_b_l, swing1LimitDegrees=15, swing2LimitDegrees=15, twistLimitDegrees=15}
 {
  bone1Name=upperarm_l
  bone2Name=spine_05
  swing1LimitDegrees=45
  swing1Motion=Free
  swing2LimitDegrees=45
  swing2Motion=Free
  twistLimitDegrees=45
  twistMotion=Free
 }
 {
  bone1Name=upperarm_r
  bone2Name=spine_05
  swing1LimitDegrees=45
  swing1Motion=Free
  swing2LimitDegrees=45
  swing2Motion=Free
  twistLimitDegrees=45
  twistMotion=Free
 }
]
